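; Whisper Remedy Sour crashdown actor: a hidden skeletal mesh appears near its origin projectile, dives via montage after a delay, then destroys itself.
; During the dive, timed traces step downward, destroying eligible low-health objects in its path.
; A Crash gameplay event triggers landing cues, including water, and a targeting-preset blast.
; The blast removes the projectile, kills weak tagged buildings, applies team-dependent GameplayEffect damage and triggers flopper spawns.
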(event EventBeginPlay
  (CallFunction|SetTeam (Team|GetActorTeam (Variables|Actor|GetInstigator)))
  (Rendering|SetVisibility (Variables|Default|GetSkeletalMesh) false true)
  (Utilities|Time|SetTimerbyEvent (EventDispatchers|CreateEvent :!event Custom|DoCrashdown) (ScalableFloat|GetValueatLevel (Variables|Default|GetHF_DiveDelay)))
  (Ability|Async|WaitGameplayEventToActor 0 "(TagName=\"Event.WhisperRemedy.Sour.Crash\")" true
    (:EventReceived
      (CallFunction|DoLanding)
      (Utilities|Time|ClearandInvalidateTimerbyHandle (Variables|Default|GetDestructionTraceTimer))
      (SoundIndicator|StartSoundIndicatorLooping (Variables|Default|GetFortSoundIndicator) (Transformation|GetActorLocation) "/WhisperRemedyAssets/Assets/Sound/WhisperRemedySour_Animal_A.WhisperRemedySour_Animal_A" 1.0 0))))

(event Custom|DoCrashdown
  (bind _skeletalmesh (Variables|Default|GetSkeletalMesh))
  (Default|UpdateLocation)
  (Transformation|SetActorRotation (CallFunction|MakeRotation))
  (Animation|Montage|PlayMontage _skeletalmesh (Variables|Default|GetMontageToPlay)
    (:then
      (Utilities|FlowControl|Delay 0.1)
      (Rendering|SetVisibility _skeletalmesh true true)
      (bind _returnvalue (Utilities|Time|SetTimerbyEvent (EventDispatchers|CreateEvent :!event TraceDestroyObjects) (Math|Float|SafeDivide 0.9 (ScalableFloat|GetValueatLevel (Variables|Default|GetHF_MaxDestructTraces))) true))
      (Variables|Default|SetDestructionTraceTimer _returnvalue)
      (CallFunction|TraceDestroyObjects))
    (:OnCompleted
      _)
    (:OnBlendOut
      _))
  (Rendering|SetVisibility (Variables|Default|GetSkeletalMesh) false true)
  (SoundIndicator|StopSoundIndicatorLooping (Variables|Default|GetFortSoundIndicator))
  (Actor|DestroyActor))

(event Custom|DoLanding
  (bind _landedonwater (Variables|Default|GetLandedOnWater?))
  (bind _returnvalue (select _landedonwater "/Game/FortressPhysicalMaterials/Water.Water" "/Water/DefaultPhysMat.DefaultPhysMat"))
  (bind _returnvalue_2 (Ability|GameplayCue|MakeGameplayCueParameters :NormalizedMagnitude 0.0 :RawMagnitude 0.0 :Location "0,0,0" :Normal "0,0,0" :PhysicalMaterial _returnvalue :GameplayEffectLevel 1 :AbilityLevel 1 :bReplicateLocationWhenUsingMinimalRepProxy false))
  (bind _originprojectile (Variables|Default|GetOriginProjectile))
  (Utilities|FlowControl|DoOnce false
    (:Completed
      (GameplayCue|ExecuteGameplayCueOnActor(Burst) self (Variables|Default|GetGC_Landing) _returnvalue_2)
      (Utilities|FlowControl|Sequence
        (:then_0
          (Utilities|Time|ClearandInvalidateTimerbyHandle (Variables|Default|GetDestructionTraceTimer))
          (bind (_asynctaskproxy _targetinghandle) (Class|AsyncActionPerformTargeting|PerformTargetingAsyncAction 0 "/WhisperRemedySour/Gameplay/TP_WhisperRemedy_Crashdown_Blast.TP_WhisperRemedy_Crashdown_Blast" true)
            (:then
              (Utilities|IsValid _originprojectile
                (:Is%20Valid
                  (Rendering|SetActorHiddenInGame _originprojectile true)
                  (Actor|DestroyActor _originprojectile))))
            (:Targeted
              (Default|ProcessDamage (TargetingSystem|TargetingResults|GetTargetingResultsActors (GameInstanceSubsystems|GetTargetingSubsystem) _targetinghandle)))))
        (:then_1
          (if (Variables|Default|GetLandedOnWater?)
            (GameplayCue|ExecuteGameplayCueOnActor(Burst) self (Variables|Default|GetGC_LandingWater) _returnvalue_2)))))))

(fn CheckFlopperSpawn
  (params (Actor :type Actor%20Object%20Reference))
  (bind _returnvalue (GameplayTags|HasMatchingGameplayTag(Message) Actor "(TagName=\"FlopperSpawn.Type.World\")"))
  (if _returnvalue
    (bind _asbga_athena_flopper_spawn_world (Utilities|Casting|CastToBGA_Athena_FlopperSpawn_World Actor))
    (Class|BGAAthenaFlopperSpawnWorld|SpawnfromExplosion _asbga_athena_flopper_spawn_world (Variables|Actor|GetInstigator) self)))

(fn MakeEffectContext
  (bind _instigator (Variables|Actor|GetInstigator))
  (Utilities|IsValid _instigator
    (:Is%20Valid
      (return (GameplayEffects|MakeFortEffectContext (Ability|GetFortAbilitySystemComponent _instigator) self _instigator self 0)))
    (:Is%20Not%20Valid
      (return (Class|AthenaFunctionLibrary|MakeSimpleFortEffectContext self self)))))

(fn MakeRotation
  (return (Math|Conversions|RotationFromXVector (Math|Vector|Normalize2D(Vector) (Math|Vector|GetUnitDirection(Vector) (Transformation|GetActorLocation) (Transformation|GetActorLocation (Variables|Actor|GetInstigator)))))))

(fn ProcessDamage
  (params (Targets :type Array%20of%20Actor%20Object%20References))
  (bind _instigatorcontroller (LocalVariables|GetInstigatorController))
  (bind _instigator (Variables|Actor|GetInstigator))
  (bind _instigator_2 (Variables|Actor|GetInstigator))
  (Utilities|FlowControl|Sequence
    (:then_0
      (LocalVariables|SetInstigatorController (Pawn|GetController _instigator))
      (LocalVariables|SetASC (Ability|GetFortAbilitySystemComponent _instigator)))
    (:then_1
      (for _array_element Targets
        (bind _output_get (LocalVariables|SetCurrentHitActor _array_element))
        (if (and (Utilities|IsValid _output_get :!class /Script/Engine.KismetSystemLibrary) (!= _output_get (Variables|Default|GetFloorActor)))
          (Utilities|FlowControl|Sequence
            (:then_0
              (bind _asbuilding_actor (Utilities|Casting|CastToBuildingActor (LocalVariables|GetCurrentHitActor))
                (:then
                  (if (GameplayTags|HasAnyMatchingGameplayTags _asbuilding_actor (Variables|Default|GetBuildingActorDirectDieTags))
                    (if (and (Variables|/Script/Engine.Actor|GetbCanBeDamaged _asbuilding_actor) (not (Variables|/Script/FortniteGame.BuildingActor|GetbIsInvulnerable _asbuilding_actor)))
                      (if (< (Health|GetMaxHealth _asbuilding_actor) (ScalableFloat|GetValueatLevel (Variables|Default|GetHF_BuildingActorDirectDieHealth)))
                        ; Direct die
                        (if (>= (Health|GetMaxHealth _asbuilding_actor) (ScalableFloat|GetValueatLevel (Variables|Default|GetHF_BuildingActorSilentDieHealth)))
                          (Gameplay|Die :self _asbuilding_actor :Killer _instigatorcontroller :DamageCauser _instigator_2) ; Die
                          (else
                            (Gameplay|Die _asbuilding_actor "(TagName=\"Gameplay.Effect.InstantDeath.Silent\")" _instigatorcontroller _instigator_2) ; Silent Die
                          ))
                        (else
                          _)))
                    (else
                      _))
                  ; Apply full game effects
                  (bind _returnvalue (Ability|GameplayEffect|AssignTagSetbyCallerMagnitude (GameplayEffects|MakeOutgoingSpec (LocalVariables|GetASC) (Variables|Default|GetGE_Damage) 0.0 (CallFunction|MakeEffectContext)) "(TagName=\"SetByCaller.Damage\")" (Math|Float|SelectFloat (ScalableFloat|GetValueasDouble (Variables|Default|GetHF_DamageToTeam)) (ScalableFloat|GetValueasDouble (Variables|Default|GetHF_Damage)) (Team|OnSameTeam (Variables|Actor|GetInstigator) (LocalVariables|GetCurrentHitActor)))))
                  (GameplayEffects|ApplyGameplayEffectSpecToTarget (LocalVariables|GetASC) _returnvalue (Ability|GetFortAbilitySystemComponent (LocalVariables|GetCurrentHitActor))))
                (:CastFailed
                  _)))
            (:then_1
              (CallFunction|CheckFlopperSpawn (LocalVariables|GetCurrentHitActor)))))))))

(fn TraceDestroyObjects
  (bind (_returnvalue_x _returnvalue_y _returnvalue_z) (Math|Vector|BreakVector (Utilities|Operators|Subtract (Utilities|Operators|Add (Transformation|GetActorLocation) (Math|Vector|MakeVector 0.0 0.0 (- (LocalVariables|GetStartingHeight) (* (Math|Float|SafeDivide (LocalVariables|GetStartingHeight) (ScalableFloat|GetValueatLevel (Variables|Default|GetHF_MaxDestructTraces))) (Variables|Default|GetTraceNum))))) (Math|Vector|MakeVector 0.0 0.0 (Variables|Default|GetDestructionTraceLength)))))
  (bind (_outhits _returnvalue) (Collision|MultiLineTraceByChannel (Utilities|Operators|Add (Transformation|GetActorLocation) (Math|Vector|MakeVector 0.0 0.0 (- (LocalVariables|GetStartingHeight) (* (Math|Float|SafeDivide (LocalVariables|GetStartingHeight) (ScalableFloat|GetValueatLevel (Variables|Default|GetHF_MaxDestructTraces))) (Variables|Default|GetTraceNum))))) (Math|Vector|MakeVector _returnvalue_x _returnvalue_y (Math|Float|Max(Float) _returnvalue_z (.z (Transformation|GetActorLocation)))) "TraceTypeQuery10" false (Utilities|Array|MakeArray (Variables|Default|GetFloorActor)) "ForDuration"))
  (if _returnvalue
    (for _array_element _outhits
      (if (Utilities|AutoDestroyTagCheck (.HitActor (Collision|BreakHitResult _array_element)))
        (if (<= (.ReturnValue (Ability|Attribute|GetFloatAttribute (.HitActor (Collision|BreakHitResult _array_element)) "(AttributeName=\"MaxHealth\",Attribute=/Script/FortniteGame.FortHealthSet:MaxHealth,AttributeOwner=\"/Script/CoreUObject.Class'/Script/FortniteGame.FortHealthSet'\")")) (ScalableFloat|GetValueatLevel (Variables|Default|GetHF_MaxBuildHealthAutoDestroy)))
          (Collision|IgnoreActorwhenMoving (Variables|Default|GetSkeletalMesh) (.HitActor (Collision|BreakHitResult _array_element)) true)
          (Actor|DestroyActor (.HitActor (Collision|BreakHitResult _array_element))))))
    (else
      _))
  (bind _result (Math|Integer|IncrementInt (Variables|Default|GetTraceNum))
    (:%20%20
      (Debugging|PrintInteger _result "" 5.0 "(R=1.000000,G=0.022174,B=0.000000,A=1.000000)"
        (:Then
          (if (>= _result (ScalableFloat|GetValueasInteger (Variables|Default|GetHF_MaxDestructTraces)))
            (Utilities|Time|ClearandInvalidateTimerbyHandle (Variables|Default|GetDestructionTraceTimer))))))))

(fn UpdateLocation
  (bind _originprojectile (Variables|Default|GetOriginProjectile))
  ; Adjust location if the bone has moved far enough away from the original spawn location (most often in water)
  (if (> (Math|Vector|Distance(Vector) (Transformation|GetActorLocation _originprojectile) (Transformation|GetActorLocation)) 5e+01)
    (Transformation|SetActorLocation (Transformation|GetWorldLocation (Actor|GetComponentbyClass _originprojectile "/Script/Engine.StaticMeshComponent")))
    (else
      _))
  (return))

(fn ConstructionScript)
