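; Decal marker actor for WhisperRemedy Sour, placed and revealed through a location broadcast event.
; On BeginPlay it switches on the owner's local net role, showing the decal for an autonomous proxy and hiding it for a simulated proxy.
; RPC_BroadcastLocation moves and rotates the actor to the given transform, then makes the decal visible regardless of role.
; Both paths use ToggleVisibility, which sets decal visibility and propagates it to child components.
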(event Custom|RPC_BroadcastLocation
  (params (Location :type Vector) (Rotation :type Rotator))
  (Transformation|SetActorLocationAndRotation Location Rotation)
  (CallFunction|ToggleVisibility true))

(event EventBeginPlay
  (switch Utilities|FlowControl|Switch|SwitchonENetRole (Networking|GetLocalRole (Actor|GetOwner))
    (:ROLE_SimulatedProxy
      (CallFunction|ToggleVisibility false))
    (:ROLE_AutonomousProxy
      (CallFunction|ToggleVisibility true))))

(fn ToggleVisibility
  (params (IsVisible :type Boolean))
  (Rendering|SetVisibility (Variables|Decal|GetDecal) IsVisible true))

(fn ConstructionScript)
