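; Thrown projectile for WhisperRemedy Sour that spawns a summon building actor (Krypto, per its comments) where it stops or enters water.
; SpawnSummon runs once, playing landing FX, then on authority spawning the BGA with floor actor, tags, origin projectile and water flag.
; Water entry adds surface-aligned impact FX, and OnBGALanded destroys the projectile.
; The hit event's landing-class and slope check leads to an empty branch.
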(event Collision|EventHit
  (params (MyComp :type Primitive%20Component%20Object%20Reference) (Other :type Actor%20Object%20Reference) (OtherComp :type Primitive%20Component%20Object%20Reference) (bSelfMoved :type Boolean) (HitLocation :type Vector :default "0, 0, 0") (HitNormal :type Vector :default "0, 0, 0") (NormalImpulse :type Vector :default "0, 0, 0") (Hit :type Hit%20Result%20Structure%20%28by%20ref%29))
  ;[0.01968 0.3956 1] Check if the hit normal is flat enough to spawn Krypto && that the surface is a valid actor class - if so, don't do a bounce
  (|CallParentFunction MyComp Other OtherComp bSelfMoved HitLocation HitNormal NormalImpulse Hit)
  (if (and (CallFunction|FilterAllowedClasses Other) (>= (Math|Vector|DotProduct HitNormal "0.000000,0.000000,1.000000") (ScalableFloat|GetValueatLevel (Variables|Default|GetHF_MinAllowedDot))))
    _))

(event EventOnStop
  (params (Hit :type Hit%20Result%20Structure%20%28by%20ref%29))
  (bind (_bblockinghit _binitialoverlap _time _distance _location _impactpoint _normal _impactnormal _physmat _hitactor) (Collision|BreakHitResult Hit))
  ;[0.01968 0.3956 1] OnStop
  (|CallParentFunction Hit)
  (CallFunction|SpawnSummon _impactpoint _hitactor)
  (GameplayCue|ExecuteGameplayCueOnActor(Burst) self (Variables|Default|GetGC_LandMark) (Ability|GameplayCue|MakeGameplayCueParameters 0.0)))

(event OnEnterWaterEvent(WaterInteractionComponent)
  (params (WaterBody :type Fort%20Water%20Body%20Actor%20Object%20Reference) (WaterInteractionComponent :type Fort%20Water%20Interaction%20Component%20Object%20Reference) (bIsFirstBody :type Boolean))
  ;[0.01968 0.3956 1] Water Interactions
  (bind _outwatersurfaceinfo (WaterBody|GetWaterSurfaceInfoStruct WaterBody (Transformation|GetActorLocation) true false false))
  (bind _returnvalue (Effects|Components|ParticleSystem|SpawnFXAtLocation (Variables|Default|GetWaterImpactFX) (Utilities|Struct|BreakWaterSurfaceInfo _outwatersurfaceinfo) (Math|Rotator|MakeRotfromX (.WaterPlaneNormal (Utilities|Struct|BreakWaterSurfaceInfo _outwatersurfaceinfo)))))
  (Effects|Components|ParticleSystem|SetFloatParameter _returnvalue "WaterDepth" (.WaterDepth (Utilities|Struct|BreakWaterSurfaceInfo _outwatersurfaceinfo)))
  (Effects|Components|ParticleSystem|SetFloatParameter _returnvalue "WaterBodyIndex" (Math|Conversions|ToFloat(Integer) (.WaterBodyIdx (Utilities|Struct|BreakWaterSurfaceInfo _outwatersurfaceinfo))))
  (CallFunction|SpawnSummon (Utilities|Struct|BreakWaterSurfaceInfo _outwatersurfaceinfo) WaterBody true))

(fn FilterAllowedClasses
  (params (Actor :type Actor%20Object%20Reference))
  (bind _array_element (Utilities|Array|ForEachLoopwithBreak (Variables|Default|GetAllowedLandingClasses))
    (:LoopBody
      (if (Utilities|IsA (Variables|GetActor) _array_element)
        (LocalVariables|SetLocalBool true)))
    (:Completed
      _))
  (return (LocalVariables|GetLocalBool)))

(fn OnBGALanded
  (Rendering|SetActorHiddenInGame false)
  (Actor|DestroyActor))

(fn SpawnSummon
  (params (Location :type Vector) (HitActor :type Actor%20Object%20Reference) (FromWater? :type Boolean))
  (if (not (Variables|Default|GetHasSpawned?))
    (Niagara|SpawnSystematLocation "/WhisperRemedySour/Assets/FX/NS_WhisperRenedy_LandMark.NS_WhisperRenedy_LandMark" Location)
    (Components|Activation|Deactivate (Variables|Default|GetRotatingMovement))
    (Variables|Default|SetHasSpawned? true)
    (Networking|SwitchHasAuthority
      (:Authority
        (bind _returnvalue (Building|SpawnBuildingGameplayActor "/WhisperRemedySour/Gameplay/Actors/BGA_WhisperRemedy_Sour.BGA_WhisperRemedy_Sour_C" (Math|Conversions|ToTransform(Vector) Location) (Variables|Actor|GetInstigator) true))
        (Variables|/WhisperRemedySour/Gameplay/Actors/BGA_WhisperRemedy_Sour.BGA_WhisperRemedy_Sour_C|SetFloorActor HitActor (Utilities|Casting|CastToBGA_WhisperRemedy_Sour _returnvalue))
        (Ability|GameplayEffect|AddLooseGameplayTags (Utilities|Casting|CastToBGA_WhisperRemedy_Sour _returnvalue) (Variables|Default|GetTagsToPass) true)
        (Variables|/WhisperRemedySour/Gameplay/Actors/BGA_WhisperRemedy_Sour.BGA_WhisperRemedy_Sour_C|SetOriginProjectile self (Utilities|Casting|CastToBGA_WhisperRemedy_Sour _returnvalue))
        (Variables|/WhisperRemedySour/Gameplay/Actors/BGA_WhisperRemedy_Sour.BGA_WhisperRemedy_Sour_C|SetLandedOnWater? (Variables|GetFromWater?) (Utilities|Casting|CastToBGA_WhisperRemedy_Sour _returnvalue))))))

(fn ConstructionScript)
