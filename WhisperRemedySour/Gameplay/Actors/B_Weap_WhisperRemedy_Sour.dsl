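; Weapon Blueprint for WhisperRemedy Sour, which aims and places a beacon, supplying its reticle color and HUD key prompts.
; GetReticleColorOverride picks between has-target and no-target colors based on the valid-target flag.
; UpdateAimingHUD finds the instigator's child tag under the aiming tag and switches on primary versus secondary aiming to choose place, cancel or release-to-act wording, with "Aim Beacon" as the fallback.
; A Sequence then writes both texts to the HUD key descriptions.
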(fn GetReticleColorOverride
  (params (HitResults :type Array%20of%20Hit%20Result%20Structures))
  (return (select (Variables|Default|GetHasValidTarget?) (Variables|Default|GetColorHasTarget) (Variables|Default|GetColorNoTarget)) false))

(fn UpdateAimingHUD
  (Utilities|FlowControl|Sequence
    (:then_0
      (bind (_outchildtag _returnvalue) (Fortnite|Utility|FindFirstTagforParentTagChecked (Variables|Actor|GetInstigator) (Variables|Default|GetAimingTag)))
      (if _returnvalue
        (switch Utilities|FlowControl|Switch|SwitchonGameplayTag _outchildtag
          (:Default
            _)
          (:Granted.WhisperRemedy.Sour.Aiming.Primary
            (LocalVariables|SetPrimaryText "NSLOCTEXT(\"[BAA09072AD138D5F62DE066D0C4F1F3E]\", \"0B47BF5F498253B22BF3F3B3EF50049B\", \"(Release) Place Beacon\")")
            (LocalVariables|SetSecondaryText "NSLOCTEXT(\"[BAA09072AD138D5F62DE066D0C4F1F3E]\", \"D22351F54B035619813A7E98F5212149\", \"Cancel\")"))
          (:Granted.WhisperRemedy.Sour.Aiming.Secondary
            (LocalVariables|SetPrimaryText "NSLOCTEXT(\"[BAA09072AD138D5F62DE066D0C4F1F3E]\", \"A9C6BC514600FFB34A8EBB8BB9E05972\", \"Place Beacon\")")
            (LocalVariables|SetSecondaryText "NSLOCTEXT(\"[BAA09072AD138D5F62DE066D0C4F1F3E]\", \"7888EC1E4004DF867AB683B709C06BEC\", \"(Release) Cancel\")")))
        (else
          _))
      (LocalVariables|SetPrimaryText "NSLOCTEXT(\"[BAA09072AD138D5F62DE066D0C4F1F3E]\", \"D0F2E87345B74FA977296BAAB134A16A\", \"Aim Beacon\")")
      (LocalVariables|SetSecondaryText "NSLOCTEXT(\"[BAA09072AD138D5F62DE066D0C4F1F3E]\", \"3C8284254DC240E005C4FD844497C964\", \"Aim Beacon\")"))
    (:then_1
      (WeaponHUDData|SetHudKeyActionDescription 0 (LocalVariables|GetPrimaryText))
      (WeaponHUDData|SetHudKeyActionDescription 1 (LocalVariables|GetSecondaryText))
      (return))))

(fn ConstructionScript)
